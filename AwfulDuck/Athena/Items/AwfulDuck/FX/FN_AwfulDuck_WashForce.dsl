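; Field system actor for AwfulDuck wash FX that applies a radial push to physics objects around its location.
; Each tick, if r.PhysicsField.EnableField is 1, it adds a transient linear velocity field: a radial vector scaled by Strength, multiplied by logarithmic falloff over Radius.
; A setter exposes Radius; ComputeStrength calculates time-based decay from StopTime but is not called here.
; EndPlay and Destroyed remove construction fields, with Destroyed also printing a debug string.
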(event EventTick
  (params (DeltaSeconds :type Float%20%28single-precision%29))
  (bind _returnvalue (Development|GetConsoleVariableIntValue "r.PhysicsField.EnableField"))
  (if (== _returnvalue 1)
    (Field|AddTransientField (Variables|Field|GetFieldSystemComponent) true "Field_LinearVelocity" 0 (Field|SetOperatorField (Variables|Default|GetMultiply) 1.0 (Field|SetRadialFalloff (Variables|Default|GetRadialFalloff) 1.0 0.0 1.0 0.0 (Variables|Forces|GetRadius) (Transformation|GetWorldLocation (Variables|Field|GetFieldSystemComponent)) "Field_Falloff_Logarithmic") (Field|SetRadialVector (Variables|Default|GetRadialVector) (Variables|Forces|GetStrength) (Transformation|GetWorldLocation (Variables|Field|GetFieldSystemComponent))) "Field_Multiply"))))

(event EventEndPlay
  (params (EndPlayReason :type EEndPlayReason%20Enum :default "Destroyed"))
  (Field|RemoveConstructionFields (Variables|Field|GetFieldSystemComponent)))

(event EventDestroyed
  (Development|PrintString "I am being destroyed" true true "(R=0.000000,G=0.660000,B=1.000000,A=1.000000)" 5.0)
  (Field|RemoveConstructionFields (Variables|Field|GetFieldSystemComponent)))

(fn ComputeStrength
  ;[1 1 1] This computes a decay in strength and kicks out a boolean for when this actor should destroy itself.
  (return (== (Math|Float|Clamp(Float) (/ (- (Utilities|Time|GetGameTimeinSeconds) (Variables|Forces|GetStop%20TIme)) 5.5)) 0.0) (Math|Float|Lerp (Variables|Forces|GetStrength) (* (- 1.0 (Math|Float|Clamp(Float) (/ (- (Utilities|Time|GetGameTimeinSeconds) (Variables|Forces|GetStop%20TIme)) 5.5))) (Variables|Forces|GetStrength)) (Math|Float|Clamp(Float) (Variables|Forces|GetStop%20TIme)))))

(fn SertRadius
  (params (Radius :type Float%20%28double-precision%29))
  (Variables|Forces|SetRadius Radius))

(fn ConstructionScript)
