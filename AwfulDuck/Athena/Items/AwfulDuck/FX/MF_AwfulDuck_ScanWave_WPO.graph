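# Material function for the AwfulDuck scan wave effect, producing an animated sine band for world position offset along vertex normals.
# World position relative to the object is tiled, offset by time and masked to the vertical axis before a sine.
# The sine is biased, sharpened by a hardness power and scaled by intensity.
# Outputs are the scalar wave and that wave multiplied by the world-space vertex normal.
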MaterialFunction /AwfulDuck/Athena/Items/AwfulDuck/FX/MF_AwfulDuck_ScanWave_WPO
graph {nodes [
  n1 = Multiply(A=Subtract(A=WorldPosition().XYZ, B=ObjectPositionWS()), B=ScalarParameter() {defaultValue=-0.0571, parameterName=WPO_SineWave_Tiling})
  n2 = Multiply(A=Time(), B=ScalarParameter() {defaultValue=3, parameterName=WPO_SineWave_PanSpeed})
  n3 = Sine(None=ComponentMask(None=Add(A=@n1, B=@n2)) {b=true})
  n4 = Multiply(
   A=Power(Base=ConstantBiasScale(None=@n3), Exp=ScalarParameter() {defaultValue=17.0436, parameterName=WPO_SineWave_Hardness})
   B=ScalarParameter() {defaultValue=2, parameterName=WPO_SineWave_Intensity}
  )
  FunctionOutput(None=Multiply(A=@n4, B=VertexNormalWS())) {outputName=SineWave_VertextNormal}
  FunctionOutput(None=@n4) {outputName=SineWave}
 ]}
