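# Material function that turns an AwfulDuck character's material into a glowing rez/de-rez look for the death transition.
# It keeps the base normal and WPO, blacks out base color, and builds emissive from blue plus a rim-driven DeRezEdgeGlow.
# Panning, time-pulsed caustics and the ScanWave function modulate the glow, with ScanWave also displacing vertices.
# Forward and mobile paths drop the caustics and scan wave for simple rim glow.
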MaterialFunction /AwfulDuck/Athena/Items/AwfulDuck/FX/MF_AwfulDuck_CharacterRez
graph {
 properties {description=Base Colors for the death Transition}
 nodes [
  n1 = Constant()
  n2 = VectorParameter() {parameterName=Blue, defaultValue=<rgba 0.1273 0.0653 0.4 1>}
  n3 = Multiply(A=@n2.RGB) {constB=3}
  n4 = FunctionInput(Preview=Fresnel() {baseReflectFraction=0, exponent=1}) {bUsePreviewValueAsDefault=true, inputName=Rim, inputType=FunctionInput_Scalar}
  n5 = Add(
   A=ShadingPathSwitch(Default=@n3, Deferred=@n3, Forward=@n2.RGB, Mobile=@n2.RGB)
   B=Multiply(A=Power(Base=@n4) {constExponent=3}, B=VectorParameter() {parameterName=DeRezEdgeGlow, defaultValue=<rgba 1.3493 6 2.2164 1>}.RGB)
  )
  n6 = TextureCoordinate()
  n7 = Constant2Vector() {g=8}
  n8 = TextureSampleParameter2D(UVs=Panner(Coordinate=Multiply(A=@n6, B=@n7)) {speedY=2}) {parameterName=Param, texture=&T_ShieldCaustics}
  n9 = Time()
  n10 = ConstantBiasScale(None=Sine(None=Add(A=@n8.R, B=@n9)))
  n11 = Multiply(A=@n8.R) {constB=3}
  n12 = Add(A=Multiply(A=Power(Base=@n5), B=Multiply(A=@n10, B=@n11)), B=Multiply(A=@n5) {constB=0.3})
  n13 = Add(
   A=@n12
   B=Multiply(A=@n12, B=MaterialFunctionCall() {materialFunction=&/MF_AwfulDuck_ScanWave_WPO}.SineWave)
  )
  n14 = GetMaterialAttributes(None=FunctionInput(Preview=SetMaterialAttributes!(MaterialAttributes)) {bUsePreviewValueAsDefault=true, inputName=BaseMaterial, inputType=FunctionInput_MaterialAttributes}) {attributeGetTypes [#1, #2]}
  n15 = MaterialFunctionCall() {materialFunction=&MF_AwfulDuck_ScanWave_WPO}
  n16 = Constant()
  n17 = SetMaterialAttributes!(
   MaterialAttributes
   BaseColor=@n1
   Metallic=@n1
   Specular=Constant() {r=1}
   Roughness=Constant() {r=0.2}
   EmissiveColor=ShadingPathSwitch(Default=@n13, Deferred=@n13, Forward=@n5, Mobile=@n5)
   Opacity
   OpacityMask
   Anisotropy
   Normal=@n14.Normal
   WorldPositionOffset=Add(
    A=MaterialFunctionCall!(WPO=@n14.WorldPositionOffset) {materialFunction=/Game/Packages/Fortress_SharedMaterials/Base_Material_Functions/Fortnite_HideElementsWithVertexColor}.Result
    B=ShadingPathSwitch(Default=@n15.SineWave_VertextNormal, Deferred=@n15.SineWave_VertextNormal, Forward=@n16, Mobile=@n16)
   )
   WorldDisplacement
   TessellationMultiplier
   SubsurfaceColor=Constant()
   CustomData0=Constant()
  ) {attributeSetTypes [
    #3
    #4
    #5
    #6
    #7
    #8
    #9
    #10
    #1
    #2
    #11
    #12
    #13
    #14
   ]}
  FunctionOutput(None=@n17)
  n18 = Clamp(None=@n4)
  n19 = OneMinus(None=@n18)
  n20 = Power(Base=@n18)
  n21 = Multiply(A=Add(A=Power(Base=@n19) {constExponent=4}, B=LinearInterpolate(Alpha=@n20) {constA=0.3, constB=0.9})) {constB=0.6}
  QualitySwitch(Default=@n4, Low=@n4, High=@n4, Medium=@n4, Epic=@n21)
  ShadingPathSwitch(Default=@n21, Deferred=@n21, Forward=@n4, Mobile=@n4)
 ]
}
