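# Masked material for the AwfulDuck character rez effect, adding glowing shield-pattern blocks to the MF_AwfulDuck_CharacterRez base as it dissolves.
# DissolveGradientPercentage drives a pre-skinned height threshold that pushes vertices outward along normals plus a smear vector.
# Scrolling world-aligned noise breaks up that threshold for dithered opacity and fake-chromatic edge glow.
# Forward and Mobile paths skip displacement, using CharacterResInEffect with panning noise for opacity.
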Material /AwfulDuck/Athena/Items/AwfulDuck/FX/M_AwfulDuck_CharacterRez
graph {
 properties {
  bAutomaticallySetUsageInEditor=false
  bUseMaterialAttributes=true
  bUsedWithClothing=true
  bUsedWithMeshDeformer=true
  bUsedWithMorphTargets=true
  bUsedWithSkeletalMesh=true
  bUsedWithSplineMeshes=true
  bUsedWithStaticLighting=true
  blendMode=BLEND_Masked
  pixelDepthOffsetMode=PDOM_Legacy
  translucencyLightingMode=TLM_VolumetricDirectional
  usedShadingModels=Default Lit
 }
 nodes [
  n1 = GetMaterialAttributes(None=MaterialFunctionCall!(BaseMaterial, Rim) {materialFunction=&/MF_AwfulDuck_CharacterRez}.Result) {attributeGetTypes [#1]}
  n2 = Constant() {r=1}
  n3 = Constant() {r=0.5}
  n4 = ScalarParameter() {defaultValue=1, group=SpawnEffect, parameterName=DissolveGradientPercentage}
  n5 = Multiply(
   A=OneMinus(None=MaterialFunctionCall!(ThresholdValue=@n2, ThresholdRange=@n3, Alpha=@n4, TwoSidedThreshold) {materialFunction=/Engine/Functions/Engine_MaterialFunctions02/Masking/ThresholdWithRange}.Result)
   B=VectorParameter() {parameterName=BlocksColor, defaultValue=<rgba 0 0.1759 0.3 1>}.RGB
  )
  n6 = Power(
   Base=TextureSample(UVs=TextureCoordinate() {uTiling=3, vTiling=3}) {samplerSource=SSM_Wrap_WorldGroupSettings, texture=/Game/Athena/SafeZone/Textures/T_JanusShield3_M}.R
   Exp=ScalarParameter() {defaultValue=0.5, parameterName=BlocksContrast}
  ) {constExponent=0.5}
  n7 = SetMaterialAttributes(MaterialAttributes=@n1.MaterialAttributes, Emissive Color=Add(A=@n1.Emissive Color, B=Multiply(A=@n5, B=@n6))) {attributeSetTypes [#1]}
  n8 = GetMaterialAttributes(None=@n7) {attributeGetTypes [#2]}
  n9 = LinearInterpolate(B=ScalarParameter() {defaultValue=1.5, parameterName=ThresholdLeadingValue}, Alpha=OneMinus(None=@n4)) {constB=1.2}
  n10 = PreSkinnedPosition()
  n11 = SphereMask(
   A=MaterialFunctionCall(Float3=@n10) {materialFunction=&BreakOutFloat3Components}.B
   B=Constant()
   Radius=ScalarParameter() {defaultValue=180, parameterName=dist}
   Hardness=Constant()
  )
  n12 = OneMinus(None=MaterialFunctionCall(
    ThresholdValue=@n9
    ThresholdRange=ScalarParameter() {defaultValue=1, parameterName=ThresholdRange}
    Alpha=@n11
    TwoSidedThreshold=StaticBool()
   ) {materialFunction=&ThresholdWithRange}.Result)
  n13 = Multiply(A=VertexNormalWS(), B=ScalarParameter() {defaultValue=10, parameterName=ExplodeIntensity})
  n14 = Multiply(A=VectorParameter() {parameterName=ExplodeSmear, defaultValue=<rgba 0 0 50 1>}.RGB, B=@n12)
  n15 = Add(A=@n8.World Position Offset, B=Multiply(A=@n12, B=Add(A=@n13, B=@n14)))
  n16 = VertexInterpolator(VS=@n10)
  n17 = Multiply(A=Multiply(A=Time(), B=ScalarParameter() {defaultValue=40, parameterName=Speed}), B=Constant3Vector() {constant=<rgba 0 0 -1 0>})
  n18 = ComponentMask(None=MaterialFunctionCall(
    TextureObject=TextureObject() {samplerType=SAMPLERTYPE_Masks, texture=/Game/Accessories/FORT_Backpacks/M_MED_Heist/Textures/T_CrystalNoise_M}
    TextureSize=ScalarParameter() {defaultValue=150, parameterName=Size}
    WorldPosition=Add(A=@n16.PS, B=@n17)
   ) {materialFunction=&WorldAlignedTexture}.XY Texture) {r=true}
  n19 = MaterialFunctionCall(
   Transition Phase=MaterialFunctionCall(
    ThresholdValue=@n9
    ThresholdRange=ScalarParameter() {defaultValue=1, parameterName=ThresholdRange}
    Alpha=VertexInterpolator(VS=@n11).PS
    TwoSidedThreshold=StaticBool()
   ) {materialFunction=&ThresholdWithRange}.Result
   Height Texture=@n18
  ) {materialFunction=&HeightLerp}
  n20 = MaterialFunctionCall(In=@n19.Lerp Alpha No Contrast, Contrast=ScalarParameter() {defaultValue=0.5, parameterName=OpacityMaskingContrast}) {materialFunction=&CheapContrast}
  n21 = GetMaterialAttributes(None=SetMaterialAttributes!(
    MaterialAttributes=SetMaterialAttributes!(MaterialAttributes=@n8.MaterialAttributes, World Position Offset=@n15) {attributeSetTypes [#2]}
    Opacity Mask=MaterialFunctionCall(Alpha Threshold=@n20.Result) {materialFunction=&DitherTemporalAA}.Result
   ) {attributeSetTypes [#3]}) {attributeGetTypes [#1]}
  n22 = Multiply(A=@n21.Emissive Color, B=@n19.Lerp Alpha No Contrast)
  n23 = MaterialFunctionCall(Float3=@n16.PS) {materialFunction=&BreakOutFloat3Components}
  n24 = ScalarParameter() {defaultValue=0.01, parameterName=FakeChromaticLookZShift}
  n25 = Add(A=Add(A=@n18, B=Multiply(A=@n23.B, B=@n24)), B=Multiply(A=Time()) {constB=-1})
  n26 = Multiply(
   A=MaterialFunctionCall(
    ThresholdValue=ScalarParameter() {defaultValue=0.2, parameterName=FakeChromaticMaskThreshold}
    ThresholdRange=ScalarParameter() {defaultValue=0.8, parameterName=FakeChromaticMaskRange}
    Alpha=@n19.Lerp Alpha No Contrast
    TwoSidedThreshold=StaticBool() {value=true}
   ) {materialFunction=&ThresholdWithRange}.Result
   B=Multiply(
    A=TextureSample(UVs=@n25) {texture=&T_Inky_Smoke_Tile_Chromatic}.RGB
    B=ScalarParameter() {defaultValue=20, parameterName=FakeChromaticIntensity}
   )
  )
  n27 = TextureSample(UVs=Panner(Coordinate=TextureCoordinate() {uTiling=1.5, vTiling=1.5}) {speedY=0.1}) {samplerType=SAMPLERTYPE_Masks, texture=&T_CrystalNoise_M}
  n28 = MaterialFunctionCall!(Animation Gradient, Rez In Percentage) {materialFunction=/Game/Packages/Fortress_SharedMaterials/Base_Material_Functions/CharacterResInEffect}
  n29 = Add(A=Multiply(A=@n27.G, B=@n27.G), B=LinearInterpolate(Alpha=@n28.Animation With Distortion) {constA=-1, constB=2})
  n30 = SetMaterialAttributes!(MaterialAttributes=@n7, Opacity Mask=Multiply(A=Clamp(None=@n29), B=@n28.Animation With Distortion)) {attributeSetTypes [#3]}
  ShadingPathSwitch(
   Default=SetMaterialAttributes!(MaterialAttributes=@n21.MaterialAttributes, Emissive Color=Add(A=@n22, B=@n26)) {attributeSetTypes [#1]}
   Forward=@n30
   Mobile=@n30
  )
 ]
}
