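# Niagara light beam effect for the AwfulDuck item, linking the owning component to a User.BeamTarget point.
# PullUpBeam draws a wide two-point ribbon, with curves driving its width and a material parameter.
# Rings continuously spawns mesh rings oriented along the beam, computing positions that lerp from target back to owner over each lifetime.
# Time- and age-phased waveforms pulse ring size and shift color, multiplied by a fading blue-cyan gradient.
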NiagaraSystem /AwfulDuck/Athena/Items/AwfulDuck/FX/NS_AwfulDuck_LightBeam
dependencies {
 usedDataInterfaces [&NiagaraDataInterfaceCurve, &NiagaraDataInterfaceColorCurve]
 usedDynamicInputs [&Multiply_VectorByFloat]
 usedModules [&SystemState]
}
emitters {
 PullUpBeam {
  emitterUpdateScript [
   &EmitterState(Life Cycle Mode=::NewEnumerator1, Loop Behavior=::NewEnumerator1, Loop Duration Mode=::NewEnumerator1, UseLoopDelay=false)
   /Niagara/Modules/Beams/BeamEmitterSetup(Absolute Beam End=true, Beam End:Vector3f=$User.BeamTarget)
   &SpawnBurst_Instantaneous(Spawn Count=2, Use Spawn Probability=false)
  ]
  particleSpawnScript [
   &InitializeParticle(Color=<rgba 0.7793 3 2.9105 1>, Position Mode=::NewEnumerator0, Write Material Random=true)
   /Niagara/Modules/Beams/SpawnBeam()
   /Niagara/Modules/Beams/BeamWidth(Beam Twist Amount=1, Beam Width=420)
  ]
  particleUpdateScript [
   &ParticleState(Kill Particles When Lifetime Has Elapsed=false, Let Infinitely Lived Particles Die When Emitter Deactivates=true)
   &DynamicMaterialParameters(
    Index 0 Param 4=&FloatFromCurve(FloatCurve=DataInterface() {propertyValues {Curve {keys [
         {arriveTangent=4.1785, interpMode=RCIM_Cubic, leaveTangent=4.1785, tangentMode=RCTM_User, time=0, value=0}
         {arriveTangent=0.9995, interpMode=RCIM_Cubic, leaveTangent=0.9995, tangentMode=RCTM_User, time=0.2223, value=0.6825}
         {arriveTangent=-0.0405, interpMode=RCIM_Cubic, leaveTangent=-0.0405, tangentMode=RCTM_User, time=1, value=1}
        ]}}})
    Module.Param0WriteEnabled=true
    Module.Param1WriteEnabled=true
    Module.Param2WriteEnabled=true
    Module.Param3WriteEnabled=true
   )
   /Niagara/Modules/Beams/UpdateBeam()
   &SolveForcesAndVelocity(Clamp Velocity=false, Limit Acceleration=false)
   /Niagara/Modules/Beams/ScaleBeamWidth(Beam Width Scale:NiagaraFloat=&FloatFromCurve(FloatCurve=DataInterface() {propertyValues {Curve {keys [
         {arriveTangent=4.1785, interpMode=RCIM_Cubic, leaveTangent=4.1785, tangentMode=RCTM_User, time=0, value=0}
         {arriveTangent=0.9995, interpMode=RCIM_Cubic, leaveTangent=0.9995, tangentMode=RCTM_User, time=0.1, value=0.75}
         {arriveTangent=-0.0405, interpMode=RCIM_Cubic, leaveTangent=-0.0405, tangentMode=RCTM_User, time=0.5, value=0.95}
        ]}}}))
  ]
  rendererClasses [&NiagaraRibbonRendererProperties]
  renderers [{rendererClass=&NiagaraRibbonRendererProperties}]
 }
 Rings {
  emitterSpawnScript [/SetVariables(Emitter.UseParticleColorInput=1) {isSetParameters=true}]
  emitterUpdateScript [
   &EmitterState(
    Enable Distance Culling=true
    MaxDistance=80000
    Module.bUseMaxDistance=true
    Module.bUseMinDistance=false
    Scalability Mode=::NewEnumerator1
   )
   &SpawnRate(SpawnRate=6.5, Use Spawn Probability=false)
  ]
  particleSpawnScript [
   &InitializeParticle(
    Color Mode=::NewEnumerator0
    Lifetime=4
    Mesh Scale Mode=::NewEnumerator3
    Mesh Uniform Scale=0.3
    UsePositionOffset=false
    Write Material Random=true
   )
   &InitialMeshOrientation(Orientation Axis=<0 0 1>, Use Orientation Vector=true, Use Rotation Vector=false)
  ]
  particleUpdateScript [
   &ParticleState()
   &DynamicMaterialParameters(
    Index 0 Param 1=&FloatFromCurve(FloatCurve=DataInterface() {propertyValues {Curve {keys [<at 0 0>,<at 0.1 1>,<at 1 1>]}}}, Scale Curve=0.5)
    Module.Param0WriteEnabled=true
    Module.Param1WriteEnabled=true
    Module.Param2WriteEnabled=true
    Module.Param3WriteEnabled=true
   )
   /SetVariables(
    Particles.Desaturation=0.4
    Particles.SineWave:NiagaraFloat=&Waveform(
     Global Input Position=&Multiply_Float(A=$Engine.Time, B=1.5)
     Global Phase=$Particles.NormalizedAge
     "Waveform [1]"=::NewEnumerator0
     "[1] Phase Scale"=2.5
     "[1] Phase"=$Particles.NormalizedAge
     "[1] Frequency"=0.5
     "[1] Amplitude Min/Max"=<1 0>
     "[1] Bias"=1
     "[2] Phase"=$Emitter.Age
     "[3] Phase"=$Emitter.Age
    )
   ) {isSetParameters=true}
   /SetVariables(Particles.WaveColor:LinearColor=&MakeLinearColor(
     R:NiagaraFloat=&Waveform(
      Global Input Position=&Multiply_Float(A=$Engine.Time, B=0.5)
      Global Phase=$Particles.NormalizedAge
      "Waveform [1]"=::NewEnumerator0
      "[1] Phase Scale"=15
      "[1] Phase"=$Particles.NormalizedAge
      "[1] Frequency"=0.5
      "[1] Amplitude Min/Max"=<1 0>
      "[1] Bias"=$Particles.Desaturation
      "[2] Phase"=$Emitter.Age
      "[3] Phase"=$Emitter.Age
     )
     G:NiagaraFloat=&Waveform(
      Global Input Position=&Multiply_Float(A=$Engine.Time, B=1.5)
      Global Phase=$Particles.NormalizedAge
      "Waveform [1]"=::NewEnumerator1
      "[1] Phase Scale"=3
      "[1] Phase"=$Particles.NormalizedAge
      "[1] Frequency"=1.5
      "[1] Amplitude Min/Max"=<1 0>
      "[1] Scale"=2
      "[1] Bias"=$Particles.Desaturation
      "[2] Phase"=$Emitter.Age
      "[3] Phase"=$Emitter.Age
     )
     B:NiagaraFloat=&Waveform(
      Global Input Position=&Multiply_Float(A=$Engine.Time, B=1.5)
      Global Phase=$Particles.NormalizedAge
      "Waveform [1]"=::NewEnumerator2
      "[1] Phase Scale"=5
      "[1] Phase"=$Particles.NormalizedAge
      "[1] Frequency"=0.5
      "[1] Amplitude Min/Max"=<1 0>
      "[1] Bias"=$Particles.Desaturation
      "[2] Phase"=$Emitter.Age
      "[3] Phase"=$Emitter.Age
     )
     A:NiagaraFloat=1
    )) {isSetParameters=true}
   /SetVariables(Particles.Color=&MultiplyLinearColor(
     Linear Color=$Particles.WaveColor
     Scale Factor=&MakeVector4_FromLinearColor(LinearColor=&ColorFromCurve(ColorCurve=DataInterface() {propertyValues {
         AlphaCurve {keys [{arriveTangent=-1.6501, interpMode=RCIM_Cubic, leaveTangent=-1.6501, tangentMode=RCTM_User, time=0, value=0.5}, <at 1 0 RCIM_Cubic>]}
         BlueCurve {keys [<at 0 1 RCIM_Cubic>,<at 1 1 RCIM_Cubic>]}
         GreenCurve {keys [<at 0 0.7891 RCIM_Cubic>,<at 1 0.5601 RCIM_Cubic>]}
         RedCurve {keys [<at 0 0.0778 RCIM_Cubic>,<at 1 0.1674 RCIM_Cubic>]}
        }}))
    )) {isSetParameters=true}
   /SetVariables(Particles.FinalPosition:Vector3f=&Lerp_Vector(
     A=Linked() {linkedVariable {name=Engine.Owner.Position, type=NiagaraPosition}}
     B=$User.BeamTarget
     Alpha=&FloatFromCurve(FloatCurve=DataInterface() {propertyValues {Curve {keys [
          {arriveTangent=-0.075, interpMode=RCIM_Cubic, leaveTangent=-0.075, tangentMode=RCTM_User, time=0, value=1}
          {arriveTangent=-2.811, interpMode=RCIM_Cubic, leaveTangent=-2.811, tangentMode=RCTM_User, time=1, value=0}
         ]}}})
    )) {isSetParameters=true}
   &ScaleMeshSize(Scale Factor=&MakeVector(X=$Particles.SineWave, Y=$Particles.SineWave, Z=1))
   /SetVariables(Particles.RingOrientation:Quat4f=&OrientMeshToVector(
     Look At Direction=&NormalizeVector(Vector To Normalize=&Subtract_Vector(A=$User.BeamTarget, B=Linked() {linkedVariable {name=Engine.Owner.Position, type=NiagaraPosition}}))
     Reference Axis Vector=<1 0 0>
    )) {isSetParameters=true}
  ]
  rendererClasses [&NiagaraMeshRendererProperties]
  renderers [{rendererClass=&NiagaraMeshRendererProperties, rendererIndex=0}]
 }
}
summary {
 systemName=NS_AwfulDuck_LightBeam
 emitters [
  {emitterName=PullUpBeam, rendererClasses [&NiagaraRibbonRendererProperties]}
  {emitterName=Rings, rendererClasses [&NiagaraMeshRendererProperties]}
 ]
 userVariables [
  User.BeamTarget {type=Vector3f, description="", defaultValue {struct=Vector3f, value=<8700 0 1500>}}
  User.X_Length {type=NiagaraFloat, defaultValue=1, description=""}
 ]
}
