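# Light beam burst Niagara effect for the AwfulDuck item, combining flares, a flash and ribbon tendrils.
# Single-particle emitters draw layered flare sprites flickered by User.FlareFlicker, an expanding, fading mesh, and a probability-gated shrinking glow.
# RibbonTest builds seven wavy ribbons along Hermite splines from the system origin toward a torus ring.
# TendrilsEndPoints places pulsing sprites with lights on a matching ring, computed independently because the particle-reader link is disabled.
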NiagaraSystem /AwfulDuck/Athena/Items/AwfulDuck/FX/NS_AwfulDuck_LightBeam_Burst
dependencies {
 usedDataInterfaces [
  &NiagaraDataInterfaceCurve
  &NiagaraDataInterfaceColorCurve
  &NiagaraDataInterfaceParticleRead
  &NiagaraDataInterfaceArrayFloat
 ]
 usedDynamicInputs []
 usedModules [&SystemState]
}
emitters {
 ConeLight {
  emitterSpawnScript [/SetVariables(Emitter.Distortion_Input=1, Emitter.LightConeOverallSize=2.2) {isSetParameters=true, module=SetVariables_#1}]
  emitterUpdateScript [
   &EmitterState(Life Cycle Mode=::NewEnumerator1, Loop Behavior=::NewEnumerator1, Loop Duration Mode=::NewEnumerator1, UseLoopDelay=false)
   &SpawnBurst_Instantaneous(Spawn Count=1, Use Spawn Probability=false)
  ]
  particleSpawnScript [
   /SetVariables(Particles.LensColor:LinearColor=<rgba 0.1227 0.2 0.1819 1>, Particles.LensSize:Vector2f=<20000 2000>) {isSetParameters=true}
   /SetVariables(Particles.DarkRadialColor:LinearColor=<rgba 0.0388 0.0328 0.1615 2>, Particles.DarkRadialSize:Vector2f=<15000 10000>) {isSetParameters=true}
   /SetVariables(
    Particles.InnerConeSize:Vector3f=&Multiply_VectorByFloat(Vector=<20 0.7 0.7>, Float=$Emitter.LightConeOverallSize)
    Particles.OuterConeSize:Vector3f=&Multiply_VectorByFloat(Vector=<30 2 2>, Float=$Emitter.LightConeOverallSize)
    Particles.TopConeSize:Vector3f=&Multiply_VectorByFloat(Vector=<0.5 9 9>, Float=$Emitter.LightConeOverallSize)
   ) {isSetParameters=true, module=SetVariables_#2}
   /SetVariables(Particles.DyParam_Inner:Vector4f=<1 -0.5 1 -1>, Particles.DyParam_Outer:Vector4f=<2 5 4 5>, Particles.DyParam_Top:Vector4f=<2 0.5 4 0.5>) {isSetParameters=true, module=SetVariables_#3}
   &InitializeParticle(
    Color=<rgba 0.2529 1 0.094 1>
    Lifetime=5
    Mesh Scale=$Particles.OuterConeSize
    Mesh Scale Mode=::NewEnumerator4
    Sprite Size=<12000 7000>
    Sprite Size Mode=::NewEnumerator4
    UsePositionOffset=false
    Write Material Random=true
   )
   &InitialMeshOrientation(Orientation Axis=<-1 0 0>, Use Orientation Vector=true, Use Rotation Vector=false)
  ]
  particleUpdateScript [
   &ParticleState(Kill Particles When Lifetime Has Elapsed=false, Let Infinitely Lived Particles Die When Emitter Deactivates=true)
   &ScaleSpriteSize(Scale Factor=&Vector2DFromFloat(Value=&Lerp_Float(
      A=1
      B=&RandomRangeFloat(Minimum=0.8, Maximum=1, Evaluation Type=::NewEnumerator1)
      Alpha=$User.FlareFlicker
     )))
  ]
  rendererClasses [&NiagaraSpriteRendererProperties]
  renderers [
   {rendererClass=&NiagaraSpriteRendererProperties}
   {rendererClass=&NiagaraSpriteRendererProperties, rendererIndex=1}
   {rendererClass=&NiagaraSpriteRendererProperties, rendererIndex=2}
  ]
 }
 ConeLight001 {
  emitterSpawnScript [/SetVariables(Emitter.LightConeOverallSize=2.2, Emitter.RingUV_Distortion_Input=1, Emitter.UseParticleColorInput=1) {isSetParameters=true, module=SetVariables_#1}]
  emitterUpdateScript [
   &EmitterState(Life Cycle Mode=::NewEnumerator1, Loop Behavior=::NewEnumerator1, Loop Duration=1, UseLoopDelay=false)
   &SpawnBurst_Instantaneous(Spawn Count=1, Use Spawn Probability=false)
  ]
  particleSpawnScript [
   /SetVariables(Particles.TopConeSize:Vector3f=&Multiply_VectorByFloat(Vector=<0.1 9 9>, Float=$Emitter.LightConeOverallSize)) {isSetParameters=true, module=SetVariables_#2}
   /SetVariables(Particles.DyParam_Top:Vector4f=<2 0.5 4 0.5>) {isSetParameters=true, module=SetVariables_#3}
   &InitializeParticle(Color=<rgba 8.8925 10 3.1116 1>, Lifetime=1.3, UsePositionOffset=false, Write Material Random=true)
   &InitialMeshOrientation(Orientation Axis=<-1 0 0>, Use Orientation Vector=true, Use Rotation Vector=false)
  ]
  particleUpdateScript [
   &ParticleState()
   &ScaleColor(
    Linear Color Curve=DataInterface() {propertyValues {
      AlphaCurve {keys [<at 0 1>,<at 1 0>]}
      BlueCurve {keys [<at 0 1>,<at 0.6241 1>]}
      GreenCurve {keys [<at 0 0.9558>,<at 0.6241 0.2958>]}
      RedCurve {keys [<at 0 0.9411>,<at 0.6241 0.0423>]}
     }}
    Scale Mode=::NewEnumerator2
   )
   /SetVariables(Particles.SizeUpdate:Vector3f=&Multiply_Vector(
     A=$Particles.TopConeSize
     B=&VectorFromFloat(Value=&FloatFromCurve(FloatCurve=DataInterface() {propertyValues {Curve {keys [<at 0 0>,<at 1 1>]}}}, Scale Curve=3))
    )) {isSetParameters=true}
  ]
  rendererClasses [&NiagaraMeshRendererProperties]
  renderers [{rendererClass=&NiagaraMeshRendererProperties, rendererIndex=0}]
 }
 FlashGLow {
  emitterUpdateScript [
   &EmitterState(Life Cycle Mode=::NewEnumerator1, Loop Behavior=::NewEnumerator1, Loop Duration Mode=::NewEnumerator1, UseLoopDelay=false)
   &SpawnBurst_Instantaneous(Spawn Count=1, Spawn Probability=$User.BurstEffectSpawnProbability, Use Spawn Probability=true)
  ]
  particleSpawnScript [
   &InitializeParticle(Color=<rgba 0 0.7681 1 1>, Sprite Size=<7680 4480>, Sprite Size Mode=::NewEnumerator4, UsePositionOffset=false, Write Material Random=true)
   &DynamicMaterialParameters(
    Index 1 Param 1=2500
    Index 1 Param 2=1
    Index 1 Param 3=1
    Index 1 Param 4=1
    Module.Param0WriteEnabled1=true
    Module.Param1WriteEnabled1=true
    Module.Param2WriteEnabled1=true
    Module.Param3WriteEnabled1=true
    Write Parameter Index 0=false
    Write Parameter Index 1=true
   )
  ]
  particleUpdateScript [
   &ParticleState()
   &ScaleSpriteSize(Scale Factor=&Vector2DFromFloat(Value=&FloatFromCurve(FloatCurve=DataInterface() {propertyValues {Curve {keys [<at 0 1>,<at 1 0>]}}}, Scale Curve=6)))
   &ScaleColor(
    Scale Alpha=&FloatFromCurve(FloatCurve=DataInterface() {propertyValues {Curve {keys [
         {arriveTangent=9.0725, interpMode=RCIM_Cubic, leaveTangent=9.0725, tangentMode=RCTM_User, time=0, value=0}
         {arriveTangent=-2.5249, interpMode=RCIM_Cubic, leaveTangent=-2.5249, tangentMode=RCTM_User, time=0.1408, value=0.9847}
         {arriveTangent=-0.0405, interpMode=RCIM_Cubic, leaveTangent=-0.0405, tangentMode=RCTM_User, time=1, value=0}
        ]}}})
    ScaleA=true
    ScaleRGB=true
   )
  ]
  rendererClasses [&NiagaraSpriteRendererProperties]
  renderers [{rendererClass=&NiagaraSpriteRendererProperties}]
 }
 RibbonTest {
  emitterUpdateScript [
   &EmitterState(Life Cycle Mode=::NewEnumerator1, Loop Behavior=::NewEnumerator1, Loop Delay=1, Loop Duration Mode=::NewEnumerator1, UseLoopDelay=false)
   &SpawnBurst_Instantaneous(Spawn Count=35, Use Spawn Probability=false)
   /SetVariables(Emitter.EmitterReader=DataInterface() {propertyValues {EmitterBinding {bindingMode=Other, emitterName=TendrilsEndPoints}}}) {enabled=false, isSetParameters=true}
  ]
  particleSpawnScript [
   &InitializeRibbon(
    Module.Write Color=true
    Module.Write Lifetime=true
    Module.Write Mass=true
    Module.Write Material Random=true
    Module.Write Position=false
    Module.Write RibbonTwist=false
    Module.Write RibbonWidth=true
    Ribbon Width=&RandomRangeFloat(Minimum=250, Maximum=500)
   )
   &FN_SetSpawnGroup(Inclusivity=true, Normalize Spawn Group=true, Particles Per Group=5, Pattern=::NewEnumerator0, Set Ribbon ID and Ribbon Link Order=true)
   &InitializeParticle(Sprite Size Mode=::NewEnumerator3, Uniform Sprite Size=1, UsePositionOffset=false, Write Material Random=true)
  ]
  particleUpdateScript [
   &ParticleState(Kill Particles When Lifetime Has Elapsed=false, Let Infinitely Lived Particles Die When Emitter Deactivates=true)
   &TorusLocation(
    Align Torus With Axis=true
    Large Radius=3000
    MaskSpawnGroup=false
    Torus Axis=<1 0 0>
    Torus Distribution Mode:ENiagaraTorusDistributionMode=::NewEnumerator1
    U Position:NiagaraFloat=&Divide_Float(A=&MakeFloatFromInt(Int=$Particles.SpawnGroup), B=7)
   ) {module=TorusLocation001}
   &self:OuterSphereLocation(New Particle Attribute Reader:NiagaraDataInterfaceParticleRead=$Emitter.EmitterReader) {enabled=false, module=OuterSphereLocation}
   /Game/Effects/Niagara/Modules/Curves/FN_HermiteCubicSpline_v2/FN_HermiteCubicSpline(
    Compute Velocity=true
    Position Blend:NiagaraFloat=$Particles.Normalized Spawn Group
    Position End:Vector3f=$Particles.OuterSpherePosition
    Position Start:Vector3f=/Niagara/DynamicInputs/Helpers/SimulationPosition()
    Write Particles Position=true
   ) {enabled=false}
   /Game/Effects/Niagara/Modules/Curves/FN_HermiteCubicSpline_v1/FN_HermiteCubicSpline(
    Compute Direction=true
    Direction Start:Vector3f=&Multiply_Vector(
     A=&NormalizeVector(
      Vector To Normalize=&Add_Vector(A=$Output.TorusLocation001.TorusVector, B=<1500 1500 1500>)
      Safe Normalize=true
      Safe Fallback Threshold=0
      Safe Fallback Vector=<0 0 0.001>
     )
     B=<0 1 1>
    )
    Position Blend:NiagaraFloat=$Particles.Normalized Spawn Group
    Position End:Vector3f=&Add_Vector(A=$Output.TorusLocation001.TorusVector, B=<-800 0 0>)
    Position Start:Vector3f=&SimulationPosition()
    Write Particles Position=true
   ) {module=FN_HermiteCubicSpline001}
   /SetVariables(Particles.Position=&Add_Vector(
     A=$Output.FN_HermiteCubicSpline001.Position
     B=/Niagara/DynamicInputs/Vectors/RotateVector(
      Vector To Rotate:Vector3f=&MakeVector(
       X=0
       Y=&Waveform(
        Global Input Position=&Add_Float(A=$Particles.Initial.RibbonLinkOrder, B=&RandomRangeFloat(Minimum=0, Maximum=1))
        Global Amplitude Scale=&Multiply_Float(A=250, B=$Particles.Normalized Spawn Group)
        Global Clamp Max=500
        "Waveform [1]"=::NewEnumerator0
        "[1] Frequency"=1
        "[1] Amplitude Min/Max"=<0 1>
        "[1] Bias"=&Lerp_Float(A=-1, B=1, Alpha=$Particles.Normalized Spawn Group)
        "[2] Phase"=$Emitter.Age
        "[3] Phase"=$Emitter.Age
       )
       Z=&Waveform(
        Global Input Position=&Add_Float(A=$Particles.Initial.RibbonLinkOrder, B=&RandomRangeFloat(Minimum=-1, Maximum=1))
        Global Amplitude Scale=&Multiply_Float(A=250, B=$Particles.Normalized Spawn Group)
        "Waveform [1]"=::NewEnumerator1
        "[1] Frequency"=0.7
        "[1] Amplitude Min/Max"=<0 1>
        "[1] Bias"=&Lerp_Float(A=-1, B=1, Alpha=$Particles.Normalized Spawn Group)
        "[2] Phase"=$Emitter.Age
        "[3] Phase"=$Emitter.Age
       )
      )
      Yaw:NiagaraFloat=&MakeFloatFromVector(Vector=$Output.FN_HermiteCubicSpline001.Direction, Channel=::NewEnumerator0)
      Pitch:NiagaraFloat=&MakeFloatFromVector(Vector=$Output.FN_HermiteCubicSpline001.Direction, Channel=::NewEnumerator1)
      Roll:NiagaraFloat=&MakeFloatFromVector(Vector=$Output.FN_HermiteCubicSpline001.Direction, Channel=::NewEnumerator2)
      Rotation Rate:NiagaraFloat=$Engine.DeltaTime
     )
    )) {isSetParameters=true}
   /SetVariables(Particles.Position=&Lerp_Vector(
     A=$Output.FN_HermiteCubicSpline001.Position
     B=Linked() {linkedVariable {name=Particles.Position, type=NiagaraPosition}}
     Alpha=&FloatFromCurve(
      FloatCurve=DataInterface() {propertyValues {Curve {keys [
          {arriveTangent=6.9713, interpMode=RCIM_Cubic, leaveTangent=6.9713, tangentMode=RCTM_User, time=0, value=0}
          {interpMode=RCIM_Cubic, tangentMode=RCTM_User, time=0.5, value=1}
          {arriveTangent=-6.3689, interpMode=RCIM_Cubic, leaveTangent=-6.3689, tangentMode=RCTM_User, time=1, value=0}
         ]}}}
      CurveIndex=$Particles.Normalized Spawn Group
      Scale Curve=0.5
     )
    )) {isSetParameters=true}
   &ScaleRibbonWidth(Ribbon Width Scale=&FloatFromCurve(
     FloatCurve=DataInterface() {propertyValues {Curve {keys [{interpMode=RCIM_Cubic, tangentMode=RCTM_User, time=0, value=1}, {interpMode=RCIM_Cubic, tangentMode=RCTM_User, time=1, value=0.1}]}}}
     CurveIndex=$Particles.Normalized Spawn Group
     Scale Curve=3
    )) {enabled=false}
   &FN_RibbonMaterial(
    Color=&ColorFromCurve(
     ColorCurve=DataInterface() {propertyValues {
       AlphaCurve {keys [<at 0 1 RCIM_Cubic>,<at 1 1 RCIM_Cubic>]}
       BlueCurve {keys [{arriveTangent=-3.673, interpMode=RCIM_Cubic, time=0.0067, value=0}, <at 0.2583 0.6466>, <at 0.533 0>, <at 0.9262 3.5492>]}
       GreenCurve {keys [{arriveTangent=0.1965, interpMode=RCIM_Cubic, time=0.0067, value=1}, <at 0.2583 2>, <at 0.533 1>, <at 0.9262 5>]}
       RedCurve {keys [{arriveTangent=0.1301, interpMode=RCIM_Cubic, time=0.0067, value=0.1928}, <at 0.2583 0>, <at 0.533 0.1955>, <at 0.9262 0>]}
      }}
     CurveIndex=$Particles.Normalized Spawn Group
    )
    Opacity Pan U Speed:NiagaraFloat=&SelectFloatFromArray(Direct Array Index=$Particles.SpawnGroup, Float Selection Array=DataInterface() {propertyValues {FloatData [0.25,0.5,0.4,0.3,0.5,0.3]}})
    Opacity Scale U=0.5
    Scale Alpha=0.1
   )
  ]
  rendererClasses [&NiagaraRibbonRendererProperties, &NiagaraSpriteRendererProperties]
  renderers [
   {rendererClass=&NiagaraRibbonRendererProperties}
   {rendererClass=&NiagaraSpriteRendererProperties, rendererIndex=1}
  ]
 }
 TendrilsEndPoints {
  emitterUpdateScript [
   &EmitterState(Life Cycle Mode=::NewEnumerator1, Loop Behavior=::NewEnumerator1, Loop Delay=1, Loop Duration Mode=::NewEnumerator1, UseLoopDelay=false)
   &SpawnBurst_Instantaneous(Spawn Count=8, Use Spawn Probability=false)
  ]
  particleSpawnScript [
   &InitializeParticle(
    Color=<rgba 0.2 1 0.2 1>
    Sprite Size Mode=::NewEnumerator3
    Uniform Sprite Size=&RandomRangeFloat(Minimum=1000, Maximum=1200)
    UsePositionOffset=false
    Write Material Random=true
   )
   /SetVariables(Particles.RandomAxis:Vector3f=/Niagara/DynamicInputs/Random/RandomVector(
     Vector Scale:NiagaraFloat=1
     Constraint Vector:Vector3f=<0 0 1>
     Randomness Mode:ENiagaraRandomnessMode=::NewEnumerator0
     Random Seed:NiagaraInt32=0
     Fixed Random Seed:NiagaraBool=false
    )) {isSetParameters=true}
   /SetVariables(Particles.RandomVector:Vector3f=&RandomVector(
     Vector Scale:NiagaraFloat=1
     Constraint Vector:Vector3f=<0 0 1>
     Randomness Mode:ENiagaraRandomnessMode=::NewEnumerator0
     Random Seed:NiagaraInt32=0
     Fixed Random Seed:NiagaraBool=false
    )) {isSetParameters=true}
   &SphereLocation(
    MaskSpawnGroup=false
    Override Seed=false
    Sphere Radius=1000
    Surface Only=true
    Surface Only Band Thickness=0.01
    UseNonUniformScale=false
    UseOffset=false
    UseSphereOrientation=false
   ) {enabled=false}
   &TorusLocation(Align Torus With Axis=false, Large Radius=5000, MaskSpawnGroup=false, Torus Distribution Mode:ENiagaraTorusDistributionMode=::NewEnumerator1) {enabled=false}
   &TorusLocation(
    Align Torus With Axis=true
    Large Radius=3000
    MaskSpawnGroup=false
    Offset=<-800 0 0>
    Torus Axis=<1 0 0>
    Torus Distribution Mode:ENiagaraTorusDistributionMode=::NewEnumerator1
    U Position:NiagaraFloat=&ReturnNormalizedExecIndex()
   ) {module=TorusLocation001}
  ]
  particleUpdateScript [
   &ParticleState(
    Kill Particles When Lifetime Has Elapsed=false
    Let Infinitely Lived Particles Die When Emitter Deactivates=true
    Loop Particles Lifetime=true
   )
   /SetVariables(Particles.RandomAxis:Vector3f=&Lerp_Vector(A=$Particles.Initial.RandomAxis, B=$Particles.Initial.RandomVector, Alpha=$Particles.NormalizedAge)) {isSetParameters=true}
   /SetVariables(Particles.Position=&RotateVector(
     Vector To Rotate:Vector3f=Linked() {linkedVariable {name=Particles.Position, type=NiagaraPosition}}
     Yaw:NiagaraFloat=&Multiply_Float(A=$Engine.DeltaTime, B=5000)
     Pitch:NiagaraFloat=0
     Roll:NiagaraFloat=0
     Rotation Rate:NiagaraFloat=$Engine.DeltaTime
    )) {enabled=false, isSetParameters=true}
   /Niagara/Modules/Update/Velocity/VortexVelocity(
    Module.Use Influence Falloff=false
    Velocity Amount:NiagaraFloat=&UniformRangedFloat(Minimum=1000, Maximum=1500)
    Vortex Axis:Vector3f=$Particles.RandomAxis
   ) {enabled=false}
   &SolveForcesAndVelocity(Clamp Velocity=false, Limit Acceleration=false)
   /SetVariables(
    Particles.LightColor:Vector3f=&MakeVectorFromLinearColorRGB(Color=<rgba 1 1 1 1>)
    Particles.NewLinearColor:LinearColor=&MultiplyLinearColor(
     Linear Color=<rgba 1 15 15 1>
     Scale Factor=&MakeVector4_FromFloat(Float=&Waveform(
       "Waveform [1]"=::NewEnumerator0
       "[1] Frequency"=&RandomRangeFloat(Minimum=1, Maximum=1.5)
       "[1] Amplitude Min/Max"=<0 1>
       "[2] Phase"=$Emitter.Age
       "[3] Phase"=$Emitter.Age
      ))
    )
   ) {isSetParameters=true}
  ]
  rendererClasses [&NiagaraSpriteRendererProperties, &NiagaraLightRendererProperties]
  renderers [
   {rendererClass=&NiagaraSpriteRendererProperties}
   {rendererClass=&NiagaraLightRendererProperties, rendererIndex=1}
  ]
 }
}
summary {
 systemName=NS_AwfulDuck_LightBeam_Burst
 emitters [
  {emitterName=ConeLight, rendererClasses [&NiagaraSpriteRendererProperties]}
  {emitterName=FlashGLow, rendererClasses [&NiagaraSpriteRendererProperties]}
  {emitterName=ConeLight001, rendererClasses [&NiagaraMeshRendererProperties]}
  {
   emitterName=TendrilsEndPoints
   rendererClasses [&NiagaraSpriteRendererProperties, &NiagaraLightRendererProperties]
  }
  {emitterName=RibbonTest, rendererClasses [&NiagaraRibbonRendererProperties, &NiagaraSpriteRendererProperties]}
 ]
 userVariables [
  User.BeamTarget {type=Vector3f, description="", defaultValue {struct=Vector3f, value=<0 0 0>}}
  User.BurstEffectSpawnProbability {type=NiagaraFloat, defaultValue=1, description=""}
  User.FEVec {type=Vector3f, description="", defaultValue {struct=Vector3f, value=<0 0 0>}}
  User.FlareFlicker {type=NiagaraFloat, defaultValue=0, description=""}
  User.RibbonIDOffset {type=NiagaraInt32}
 ]
}
