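# Niagara warning effect for the AwfulDuck item, built from cone-shaped mesh lights and a sprite glow.
# ConeLight001 grows a fading cone and FlashGLow adds a brief sprite flash.
# ConeLightBurst widens and fades inner and outer cones, while ConeLightLoop fades in a matching pair that persists beyond its lifetime.
# User.X_Length stretches burst and loop cones along X; curves drive size and color over particle life.
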NiagaraSystem /AwfulDuck/Athena/Items/AwfulDuck/FX/NS_AwfulDuck_Warning_Local
dependencies {
 usedDataInterfaces [&NiagaraDataInterfaceColorCurve, &NiagaraDataInterfaceCurve]
 usedDynamicInputs []
 usedModules [&SystemState]
}
emitters {
 ConeLight001 {
  emitterSpawnScript [/SetVariables(Emitter.LightConeOverallSize=2.2, Emitter.RingUV_Distortion_Input=1, Emitter.UseParticleColorInput=1) {isSetParameters=true, module=SetVariables_#1}]
  emitterUpdateScript [
   &EmitterState(Life Cycle Mode=::NewEnumerator1, Loop Behavior=::NewEnumerator1, Loop Duration=1, UseLoopDelay=false)
   &SpawnBurst_Instantaneous(Spawn Count=1, Use Spawn Probability=false)
  ]
  particleSpawnScript [
   /SetVariables(Particles.TopConeSize:Vector3f=&Multiply_VectorByFloat(Vector=<0.1 9 9>, Float=$Emitter.LightConeOverallSize)) {isSetParameters=true, module=SetVariables_#2}
   /SetVariables(Particles.DyParam_Top:Vector4f=<2 0.5 4 0.5>) {isSetParameters=true, module=SetVariables_#3}
   &InitializeParticle(Color=<rgba 1.6006 1.8 0.5601 1>, Lifetime=1.3, UsePositionOffset=false, Write Material Random=true)
   &InitialMeshOrientation(Orientation Axis=<-1 0 0>, Use Orientation Vector=true, Use Rotation Vector=false)
  ]
  particleUpdateScript [
   &ParticleState()
   &ScaleColor(
    Linear Color Curve=DataInterface() {propertyValues {
      AlphaCurve {keys [<at 0 1>,<at 1 0>]}
      BlueCurve {keys [<at 0 1>,<at 0.6241 1>]}
      GreenCurve {keys [<at 0 0.9558>,<at 0.6241 0.2958>]}
      RedCurve {keys [<at 0 0.9411>,<at 0.6241 0.0423>]}
     }}
    Scale Mode=::NewEnumerator2
   )
   /SetVariables(Particles.SizeUpdate:Vector3f=&Multiply_Vector(
     A=$Particles.TopConeSize
     B=&VectorFromFloat(Value=&FloatFromCurve(FloatCurve=DataInterface() {propertyValues {Curve {keys [<at 0 0>,<at 1 1>]}}}, Scale Curve=2.5))
    )) {isSetParameters=true}
  ]
  rendererClasses [&NiagaraMeshRendererProperties]
  renderers [{rendererClass=&NiagaraMeshRendererProperties, rendererIndex=0}]
 }
 ConeLightBurst {
  emitterSpawnScript [/SetVariables(Emitter.LightConeOverallSize=2.2, Emitter.UseParticleColorInput=1) {isSetParameters=true, module=SetVariables_#1}]
  emitterUpdateScript [
   &EmitterState(Life Cycle Mode=::NewEnumerator1, Loop Behavior=::NewEnumerator1, Loop Duration=1, UseLoopDelay=false)
   &SpawnBurst_Instantaneous(Spawn Count=1, Spawn Time=0.1, Use Spawn Probability=false)
  ]
  particleSpawnScript [
   /SetVariables(
    Particles.InnerConeSize:Vector3f=&Multiply_VectorByFloat(Vector=<20 0.7 0.7>, Float=$Emitter.LightConeOverallSize)
    Particles.OuterConeSize:Vector3f=&Multiply_VectorByFloat(Vector=<30 2 2>, Float=$Emitter.LightConeOverallSize)
   ) {isSetParameters=true, module=SetVariables_#2}
   /SetVariables(Particles.DyParam_Inner:Vector4f=<1 -0.5 1 -1>, Particles.DyParam_Outer:Vector4f=<2 5 4 5>) {isSetParameters=true, module=SetVariables_#3}
   &InitializeParticle(
    Color=<rgba 0.9558 0.9669 1 1>
    Mesh Scale=$Particles.OuterConeSize
    Mesh Scale Mode=::NewEnumerator4
    Sprite Size=<12000 7000>
    Sprite Size Mode=::NewEnumerator4
    UsePositionOffset=false
    Write Material Random=true
   )
  ]
  particleUpdateScript [
   &ParticleState()
   /SetVariables(Particles.LightConeRadius:NiagaraFloat=&FloatFromCurve(FloatCurve=DataInterface() {propertyValues {Curve {keys [{arriveTangent=2.0808, interpMode=RCIM_Cubic, leaveTangent=2.0808, tangentMode=RCTM_User, time=0, value=0}, <at 0.6 1>]}}})) {isSetParameters=true}
   /SetVariables(
    Particles.InnerConeSizeUpdate:Vector3f=&Multiply_Vector(
     A=$Particles.InnerConeSize
     B=&MakeVector(X=$User.X_Length, Y=$Particles.LightConeRadius, Z=$Particles.LightConeRadius)
    )
    Particles.OuterConeSizeUpdate:Vector3f=&Multiply_Vector(
     A=$Particles.OuterConeSize
     B=&MakeVector(X=$User.X_Length, Y=$Particles.LightConeRadius, Z=$Particles.LightConeRadius)
    )
   ) {isSetParameters=true, module=SetVariables_#4}
   &ScaleColor(
    Linear Color Curve=DataInterface() {propertyValues {
      AlphaCurve {keys [<at 0 1 RCIM_Cubic>,<at 1 0 RCIM_Cubic>]}
      BlueCurve {keys [{arriveTangent=-26.2894, interpMode=RCIM_Cubic, leaveTangent=-26.2894, tangentMode=RCTM_User, time=0, value=8.7692}, <at 1 20>]}
      GreenCurve {keys [{arriveTangent=8.994, interpMode=RCIM_Cubic, leaveTangent=8.994, tangentMode=RCTM_User, time=0, value=7.2934}, <at 1 0>]}
      RedCurve {keys [{arriveTangent=-4.9117, interpMode=RCIM_Cubic, leaveTangent=-4.9117, tangentMode=RCTM_User, time=0, value=0}, <at 1 0.2141>]}
     }}
    Scale Mode=::NewEnumerator2
   )
  ]
  rendererClasses [&NiagaraMeshRendererProperties]
  renderers [
   {rendererClass=&NiagaraMeshRendererProperties, rendererIndex=0}
   {rendererClass=&NiagaraMeshRendererProperties, rendererIndex=1}
  ]
 }
 ConeLightLoop {
  emitterSpawnScript [/SetVariables(Emitter.LightConeOverallSize=2.2, Emitter.UseParticleColorInput=1) {isSetParameters=true, module=SetVariables_#1}]
  emitterUpdateScript [
   &EmitterState(Life Cycle Mode=::NewEnumerator1, Loop Behavior=::NewEnumerator1, Loop Duration Mode=::NewEnumerator1, UseLoopDelay=false)
   &SpawnBurst_Instantaneous(Spawn Count=1, Spawn Time=0.1, Use Spawn Probability=false)
  ]
  particleSpawnScript [
   /SetVariables(
    Particles.InnerConeSize:Vector3f=&Multiply_VectorByFloat(Vector=<20 0.7 0.7>, Float=$Emitter.LightConeOverallSize)
    Particles.OuterConeSize:Vector3f=&Multiply_VectorByFloat(Vector=<30 2 2>, Float=$Emitter.LightConeOverallSize)
   ) {isSetParameters=true, module=SetVariables_#2}
   /SetVariables(Particles.DyParam_Inner:Vector4f=<1 -0.5 1 -1>, Particles.DyParam_Outer:Vector4f=<2 5 4 5>) {isSetParameters=true, module=SetVariables_#3}
   &InitializeParticle(
    Lifetime=5
    Mesh Scale=$Particles.OuterConeSize
    Mesh Scale Mode=::NewEnumerator4
    Sprite Size=<12000 7000>
    Sprite Size Mode=::NewEnumerator4
    UsePositionOffset=false
    Write Material Random=true
   )
  ]
  particleUpdateScript [
   &ParticleState(Kill Particles When Lifetime Has Elapsed=false, Let Infinitely Lived Particles Die When Emitter Deactivates=true)
   /SetVariables(
    Particles.InnerConeSizeUpdate:Vector3f=&Multiply_Vector(A=$Particles.InnerConeSize, B=&MakeVector(X=$User.X_Length, Y=1, Z=1))
    Particles.OuterConeSizeUpdate:Vector3f=&Multiply_Vector(A=$Particles.OuterConeSize, B=&MakeVector(X=$User.X_Length, Y=1, Z=1))
   ) {isSetParameters=true, module=SetVariables_#4}
   &ScaleColor(
    Linear Color Curve=DataInterface() {propertyValues {
      AlphaCurve {keys [<at 0.0087 0>,<at 0.2015 1>,<at 1 1>]}
      BlueCurve {keys [<at 0 1>,<at 1 1>]}
      GreenCurve {keys [<at 0 1>,<at 1 1>]}
      RedCurve {keys [<at 0 1>,<at 1 1>]}
     }}
    Scale Mode=::NewEnumerator2
   )
  ]
  rendererClasses [&NiagaraMeshRendererProperties]
  renderers [
   {rendererClass=&NiagaraMeshRendererProperties, rendererIndex=0}
   {rendererClass=&NiagaraMeshRendererProperties, rendererIndex=1}
  ]
 }
 FlashGLow {
  emitterUpdateScript [
   &EmitterState(Life Cycle Mode=::NewEnumerator1, Loop Behavior=::NewEnumerator1, Loop Duration Mode=::NewEnumerator1, UseLoopDelay=false)
   &SpawnBurst_Instantaneous(Spawn Count=1, Use Spawn Probability=true)
  ]
  particleSpawnScript [&InitializeParticle(
    Color=<rgba 0.0978 1 0.4576 1>
    Lifetime=0.3
    Sprite Size=<3882.4275 1779.0151>
    Sprite Size Mode=::NewEnumerator4
    UsePositionOffset=false
    Write Material Random=true
   )]
  particleUpdateScript [
   &ParticleState()
   &ScaleSpriteSize(Scale Factor=&Vector2DFromFloat(Value=&FloatFromCurve(
      FloatCurve=DataInterface() {propertyValues {Curve {keys [
          {arriveTangent=5.658, interpMode=RCIM_Cubic, leaveTangent=5.658, tangentMode=RCTM_User, time=0, value=0}
          {arriveTangent=-2.5249, interpMode=RCIM_Cubic, leaveTangent=-2.5249, tangentMode=RCTM_User, time=0.5, value=0.85}
          {arriveTangent=-0.0405, interpMode=RCIM_Cubic, leaveTangent=-0.0405, tangentMode=RCTM_User, time=1, value=0}
         ]}}}
      Scale Curve=6
     )))
  ]
  rendererClasses [&NiagaraSpriteRendererProperties]
  renderers [{rendererClass=&NiagaraSpriteRendererProperties}]
 }
}
summary {
 systemName=NS_AwfulDuck_Warning_Local
 emitters [
  {emitterName=ConeLightLoop, rendererClasses [&NiagaraMeshRendererProperties]}
  {emitterName=ConeLightBurst, rendererClasses [&NiagaraMeshRendererProperties]}
  {emitterName=FlashGLow, rendererClasses [&NiagaraSpriteRendererProperties]}
  {emitterName=ConeLight001, rendererClasses [&NiagaraMeshRendererProperties]}
 ]
 userVariables [
  User.BurstEffectSpawnProbability {type=NiagaraFloat, defaultValue=0, description=""}
  User.FlareFlicker {type=NiagaraFloat, defaultValue=0, description=""}
  User.X_Length {type=NiagaraFloat, defaultValue=1, description=""}
 ]
}
