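# Niagara camera-lens effect for the AwfulDuck item, named as the loop variant, that shows a bright red mesh overlay at an offset from the system with a camera offset.
# Its single FN_Mesh emitter burst-spawns one oriented mesh particle that is not killed when its lifetime elapses.
# Curves drive alpha, multiplied by User.WhiteWash, and a dynamic material parameter, while User.ViewportScale scales the mesh size.
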NiagaraSystem /AwfulDuck/Athena/Items/AwfulDuck/FX/NS_Camera_Lens_AwfulDuck_Loop
dependencies {usedDataInterfaces [&NiagaraDataInterfaceCurve], usedDynamicInputs [], usedModules [&SystemState]}
emitters {FN_Mesh {
  emitterUpdateScript [
   &EmitterState(Life Cycle Mode=::NewEnumerator1, Loop Behavior=::NewEnumerator1, Loop Duration Mode=::NewEnumerator1, UseLoopDelay=false)
   &SpawnBurst_Instantaneous(Spawn Count=1, Use Spawn Probability=false)
  ]
  particleSpawnScript [
   &InitializeParticle(Lifetime=2.5, Mesh Scale=<2 1 2>, Mesh Scale Mode=::NewEnumerator4, UsePositionOffset=false, Write Material Random=true)
   &InitialMeshOrientation(Rotation=<0 0 0.25>, Use Orientation Vector=true, Use Rotation Vector=true)
   &SystemLocation(MaskSpawnGroup=false, Offset=<-280 0 0>)
   &CameraOffset(Camera Offset Amount=58)
  ]
  particleUpdateScript [
   &ParticleState(Kill Particles When Lifetime Has Elapsed=false, Let Infinitely Lived Particles Die When Emitter Deactivates=true)
   &ScaleColor(
    Scale Alpha=&FloatFromCurve(
     FloatCurve=DataInterface() {propertyValues {Curve {keys [
         {arriveTangent=-0.25, interpMode=RCIM_Cubic, time=0, value=0}
         {arriveTangent=1.25, leaveTangent=1.25, time=0.0124, value=1}
         {interpMode=RCIM_Cubic, leaveTangent=-1.0125, time=0.8, value=1}
        ]}}}
     Scale Curve=$User.WhiteWash
    )
    Scale RGB=<15 0 0>
    ScaleA=true
    ScaleRGB=true
   ) {module=ScaleColor002}
   &DynamicMaterialParameters(
    Index 0 Param 1=&FloatFromCurve(FloatCurve=DataInterface() {propertyValues {Curve {keys [
         <at 0 0 RCIM_Cubic>
         {arriveTangent=6.6983, interpMode=RCIM_Cubic, leaveTangent=6.6983, time=0.2008, value=4.0944}
         {interpMode=RCIM_Cubic, leaveTangent=-5.1229, time=0.6048, value=4.0513}
        ]}}})
    Index 0 Param 2=1
    Index 0 Param 3=1
    Index 0 Param 4=1
    Module.Param0WriteEnabled=true
    Module.Param1WriteEnabled=true
    Module.Param2WriteEnabled=true
    Module.Param3WriteEnabled=true
   ) {module=DynamicMaterialParameters002}
   &ScaleMeshSize(Scale Factor=&VectorFromFloat(Value=$User.ViewportScale))
  ]
  rendererClasses [&NiagaraMeshRendererProperties]
  renderers [{rendererClass=&NiagaraMeshRendererProperties, rendererIndex=0}]
 }}
summary {
 systemName=NS_Camera_Lens_AwfulDuck_Loop
 emitters [{emitterName=FN_Mesh, rendererClasses [&NiagaraMeshRendererProperties]}]
 userVariables [
  User.ViewportScale {type=NiagaraFloat, defaultValue=1, description=""}
  User.WhiteWash {type=NiagaraFloat, defaultValue=1, description=""}
 ]
}
